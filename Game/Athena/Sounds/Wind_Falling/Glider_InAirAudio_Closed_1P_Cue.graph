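# First-person looping wind Sound Cue for falling with a closed glider, limited to one instance per owner.
# A quality node routes the first two levels to a layered skydive mix and the last two to a simpler closed-umbrella loop plus delayed entry sound.
# The layered mix crossfades bank, horizontal/vertical and ground-proximity wind loops by runtime parameters, alongside a cloth flap that fades in.
# The BackParam crossfade has no sound connected.
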SoundCue /Game/Athena/Sounds/Wind_Falling/Glider_InAirAudio_Closed_1P_Cue
root=QualityLevel_0
nodes [
 {id=Delay_1, children [WavePlayer_5], properties {delayMax=0.5, delayMin=0.2}}
 {
  id=Enveloper_0
  children [WavePlayer_13]
  properties {pitchCurve {editorCurveData {keys [{time=0}]}}, volumeCurve {editorCurveData {keys [<at 0 0>,<at 2 1>]}}}
 }
 {
  id=Enveloper_1
  children [WavePlayer_6]
  properties {pitchCurve {editorCurveData {keys [{time=0}]}}, volumeCurve {editorCurveData {keys [<at 0 1>,<at 5 0.75>]}}}
 }
 {id=Mixer_0, children [Enveloper_1,Delay_1], properties {inputVolume [0.8,1.5]}}
 {
  id=Mixer_2
  children [ParamCrossFade_1, ParamCrossFade_2, ParamCrossFade_3, ParamCrossFade_4, Enveloper_0, ParamCrossFade_0]
  properties {inputVolume [1,1,1,1,0.6,1]}
 }
 {id=Modulator_0, children [WavePlayer_1], properties {pitchMax=1.5, pitchMin=1.5, volumeMax=1.5, volumeMin=1.5}}
 {
  id=ParamCrossFade_0
  children [Modulator_0]
  properties {
   paramName=GroundProx
   crossFadeInput [{fadeInDistanceEnd=1, fadeInDistanceStart=0, fadeOutDistanceEnd=100, fadeOutDistanceStart=1, volume=1}]
  }
 }
 {
  id=ParamCrossFade_1
  children [WavePlayer_12]
  properties {
   paramName=RightParam
   crossFadeInput [{fadeInDistanceEnd=1, fadeInDistanceStart=0, fadeOutDistanceEnd=100, fadeOutDistanceStart=1, volume=1}]
  }
 }
 {
  id=ParamCrossFade_2
  children [WavePlayer_8]
  properties {paramName=LeftParam, crossFadeInput [{fadeInDistanceEnd=1, fadeInDistanceStart=0, fadeOutDistanceEnd=100, fadeOutDistanceStart=1, volume=1}]}
 }
 {
  id=ParamCrossFade_3
  children [WavePlayer_14, WavePlayer_15]
  properties {
   paramName=ForwardParam
   crossFadeInput [
    {fadeInDistanceEnd=0, fadeInDistanceStart=0, fadeOutDistanceEnd=1, fadeOutDistanceStart=0, volume=1}
    {fadeInDistanceEnd=1, fadeInDistanceStart=0, fadeOutDistanceEnd=100, fadeOutDistanceStart=1, volume=1}
   ]
  }
 }
 {
  id=ParamCrossFade_4
  children [null]
  properties {paramName=BackParam, crossFadeInput [{fadeInDistanceEnd=1, fadeInDistanceStart=0, fadeOutDistanceEnd=100, fadeOutDistanceStart=1, volume=1}]}
 }
 {id=QualityLevel_0, children [Mixer_2,Mixer_2,Mixer_0,Mixer_0]}
 {id=WavePlayer_1, properties {bLooping=true, soundWaveAssetPtr=&/AMB_Wind_Skydive_Vertical_Loop_01}}
 {id=WavePlayer_12, properties {bLooping=true, soundWaveAssetPtr=&/AMB_Wind_Skydive_Bank_Right_Loop_01}}
 {id=WavePlayer_13, properties {bLooping=true, soundWaveAssetPtr=&/AMB_Wind_Skydive_Cloth_Flap_Loop_01}}
 {id=WavePlayer_14, properties {bLooping=true, soundWaveAssetPtr=&/AMB_Wind_Skydive_Horizontal_Loop_01}}
 {id=WavePlayer_15, properties {bLooping=true, soundWaveAssetPtr=&AMB_Wind_Skydive_Vertical_Loop_01}}
 {id=WavePlayer_5, properties {soundWaveAssetPtr=&/sfx_parachute_skydive_loop_enter_01_opt}}
 {id=WavePlayer_6, properties {bLooping=true, soundWaveAssetPtr=&/Wind_UmbrellaClosed_Loop_01_opt}}
 {id=WavePlayer_8, properties {bLooping=true, soundWaveAssetPtr=&/AMB_Wind_Skydive_Bank_Left_Loop_01}}
]
properties {
 attenuationSettings=/Game/Sounds/Attenuation/Fort_Player_Glider_OpenClose_Athena_Att
 bOverrideConcurrency=true
 maxDistance=10000
 soundClassObject=/Game/Sounds/FortSoundClassesAndModes/Fort_Gliders_InAir_1P
 volumeMultiplier=0.6
 concurrencyOverrides {bLimitToOwner=true, maxCount=1, resolutionRule=StopOldest}
}
