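# Sound cue named for first-person spectate in-air audio of a closed glider, with an empty node graph and no root, so it plays no sound.
# It still carries settings: the first-person in-air glider sound class, the Athena glider open/close attenuation and a reduced volume.
# A concurrency override allows one instance per owner, stopping the oldest.
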SoundCue /Game/Athena/Sounds/Wind_Falling/Glider_InAirAudio_Closed_1P_Spectate_Cue
root=null
nodes []
properties {
 attenuationSettings=/Game/Sounds/Attenuation/Fort_Player_Glider_OpenClose_Athena_Att
 bOverrideConcurrency=true
 maxDistance=10000
 soundClassObject=/Game/Sounds/FortSoundClassesAndModes/Fort_Gliders_InAir_1P
 volumeMultiplier=0.4
 concurrencyOverrides {bLimitToOwner=true, maxCount=1, resolutionRule=StopOldest}
}
