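# Third-person looping in-air sound cue for a closed glider while falling in Athena.
# An enveloper fades in the output of a quality switch: the first two levels share a layered mix, and the last two use one pitched-up umbrella wind loop.
# The layered mix blends a skydive loop with pitched-up close and distant synth loops, each with its own attenuation.
# Overridden concurrency caps voices at four with distance-based volume scaling.
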SoundCue /Game/Athena/Sounds/Wind_Falling/Glider_InAirAudio_Closed_3P_Cue
root=Enveloper_0
nodes [
 {
  id=Attenuation_1
  children [WavePlayer_9]
  properties {attenuationSettings=/Game/Sounds/Attenuation/Fort_Player_Glider_CloseDetail_Athena_Att}
 }
 {
  id=Attenuation_2
  children [Modulator_0]
  properties {attenuationSettings=/Game/Sounds/Attenuation/Fort_Player_Glider_GenericInAir_Athena_Att}
 }
 {
  id=Attenuation_3
  children [Modulator_4]
  properties {attenuationSettings=/Game/Sounds/Attenuation/Fort_Player_Glider_GenericInAir_Close_Athena_Att}
 }
 {
  id=Attenuation_4
  children [Modulator_5]
  properties {attenuationSettings=/Game/Sounds/Attenuation/Fort_Player_Glider_GenericInAir_Distant_Athena_Att}
 }
 {
  id=Enveloper_0
  children [QualityLevel_0]
  properties {pitchCurve {editorCurveData {keys [{time=0}]}}, volumeCurve {editorCurveData {keys [<at 0 0>,<at 1 1>]}}}
 }
 {id=Mixer_1, children [Attenuation_3, Attenuation_4], properties {inputVolume [1.75,2.85]}}
 {id=Mixer_2, children [Attenuation_1, Mixer_1], properties {inputVolume [2,2]}}
 {id=Modulator_0, children [WavePlayer_0], properties {pitchMax=1.15, pitchMin=1.15, volumeMax=1.5, volumeMin=1.5}}
 {id=Modulator_4, children [WavePlayer_4], properties {pitchMax=1.5, pitchMin=1.5}}
 {id=Modulator_5, children [WavePlayer_5], properties {pitchMax=1.5, pitchMin=1.5}}
 {id=QualityLevel_0, children [Mixer_2, Mixer_2, Attenuation_2, Attenuation_2]}
 {id=WavePlayer_0, properties {bLooping=true, soundWaveAssetPtr=&/Wind_UmbrellaOpen_Loop_3P_opt}}
 {id=WavePlayer_4, properties {bLooping=true, soundWaveAssetPtr=&/sfx_parachute_skydive_loop_synth_3p_02}}
 {
  id=WavePlayer_5
  properties {bLooping=true, soundWaveAssetPtr=&/sfx_parachute_skydive_loop_synth_3p_distant_02}
 }
 {id=WavePlayer_9, properties {bLooping=true, soundWaveAssetPtr=&/sfx_parachute_skydive_loop_3p}}
]
properties {
 bOverrideConcurrency=true
 maxDistance=13000
 soundClassObject=&Fort_Gliders_InAir
 volumeMultiplier=1.5
 attenuationOverrides {bAttenuate=false, falloffDistance=100000}
 concurrencyOverrides {maxCount=4, voiceStealReleaseTime=0.15, volumeScaleAttackTime=0.5, volumeScaleMode=Distance}
}
