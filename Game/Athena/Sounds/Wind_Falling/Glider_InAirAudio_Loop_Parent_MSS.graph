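# Parent MetaSound for the glider's looping in-air wind audio, with its cloth flap, sideways, horizontal and vertical loops exposed as inputs.
# ForwardParam crossfades stereo horizontal and vertical wind, mixed with mono cloth and sideways layers.
# Rising GroundProx lowers cloth and stereo wind, plus a control signal, scaled by AmbienceDuckStrength, written to the outdoor-ambience sidechain bus.
# LeftParam, RightParam and BackParam are unconnected.
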MetaSoundSource /Game/Athena/Sounds/Wind_Falling/Glider_InAirAudio_Loop_Parent_MSS
inputs [
 AmbienceDuckStrength {type=Float}
 BackParam {type=Float}
 ClothFlapLoop {type=WaveAsset}
 ForwardParam {type=Float}
 GroundProx {type=Float}
 HorizontalLoop {type=WaveAsset}
 LeftParam {type=Float}
 RightParam {type=Float}
 SidewaysLoop {type=WaveAsset}
 UE.Source.OnPlay {type=Trigger}
 VerticalLoop {type=WaveAsset}
]
interfaces [UE.OutputFormat.Stereo:1.0, UE.Source:1.1]
outputs [
 "UE.OutputFormat.Stereo.Audio:0" {type=Audio}
 "UE.OutputFormat.Stereo.Audio:1" {type=Audio}
]
pages [Default {
  input_defaults {
   AmbienceDuckStrength=0.5
   BackParam=0
   ClothFlapLoop=&/AMB_Wind_Skydive_Cloth_Flap_Loop_01
   ForwardParam=0.876
   GroundProx=0
   HorizontalLoop=&/AMB_Wind_Skydive_Horizontal_Loop_01
   LeftParam=0
   RightParam=0
   SidewaysLoop=&/AMB_Wind_Skydive_Bank_Loop_01
   UE.Source.OnPlay=null
   VerticalLoop=&/AMB_Wind_Skydive_Vertical_Loop_01
  }
  nodes [
   n1 = MapRange.MapRange:Float(In=^GroundProx) {also {defaults {Out Range A=1, Out Range B=0.75}}}
   AudioBusWriter1_0 = "UE.Audio Bus Writer (1)"(In 0=UE.Multiply:Audio by Float(AdditionalOperands=^AmbienceDuckStrength, PrimaryOperand=UE.ConversionFloatToAudio(In=@n1.Out Value).Out).Out) {also {defaults {Audio Bus=/Game/Sounds/Modulation/EnvelopeFollowers/Volume/AB_SideChain_FortAmbient_Outdoors_Vol}}}
   WavePlayer_0 = UE.Wave Player:Mono(Wave Asset=^ClothFlapLoop, Play=^UE.Source.OnPlay) {also {defaults {Loop="<MetaSoundBuilderResult.SUCCEEDED: 0>"}, version [1,1]}}
   WavePlayer_1 = UE.Wave Player:Mono(Wave Asset=^SidewaysLoop, Play=^UE.Source.OnPlay) {also {defaults {Loop="<MetaSoundBuilderResult.SUCCEEDED: 0>"}, version [1,1]}}
   n2 = UE.Add:Audio(
    PrimaryOperand=UE.Multiply:Audio by Float(
     AdditionalOperands=MapRange.MapRange:Float(In=^GroundProx) {also {defaults {Out Range A=1, Out Range B=0.5}}}.Out Value
     PrimaryOperand=@WavePlayer_0.Out Mono
    ).Out
    AdditionalOperands=@WavePlayer_1.Out Mono
   )
   WavePlayer_2 = UE.Wave Player:Stereo(Wave Asset=^HorizontalLoop, Play=^UE.Source.OnPlay) {also {defaults {Loop="<MetaSoundBuilderResult.SUCCEEDED: 0>"}, version [1,1]}}
   WavePlayer_3 = UE.Wave Player:Stereo(Play=^UE.Source.OnPlay, Wave Asset=^VerticalLoop) {also {defaults {Loop="<MetaSoundBuilderResult.SUCCEEDED: 0>"}, version [1,1]}}
   CrossfadeTriggerRouteAudio2_0 = "Crossfade.Trigger Route (Audio, 2):Audio"(Crossfade Value=^ForwardParam, In 0=@WavePlayer_2.Out Left, In 1=@WavePlayer_3.Out Left)
   CrossfadeTriggerRouteAudio2_1 = "Crossfade.Trigger Route (Audio, 2):Audio"(Crossfade Value=^ForwardParam, In 0=@WavePlayer_2.Out Right, In 1=@WavePlayer_3.Out Right)
   AudioMixerAudioMixerStereo2_0 = "AudioMixer.Audio Mixer (Stereo, 2)"(In 0 L=@n2.Out, In 0 R=@n2.Out, In 1 L=@CrossfadeTriggerRouteAudio2_0.Out, In 1 R=@CrossfadeTriggerRouteAudio2_1.Out, Gain 1=@n1.Out Value)
   InitVariable.WaveAsset()
   InitVariable.WaveAsset()
  ]
  links [
   {dst="@UE.OutputFormat.Stereo.Audio:0", src=AudioMixerAudioMixerStereo2_0.Out L}
   {dst="@UE.OutputFormat.Stereo.Audio:1", src=AudioMixerAudioMixerStereo2_0.Out R}
  ]
 }]
properties {
 assetImportData=&self:AssetImportData
 bLooping=true
 duration=10000
 numChannels=2
 outputFormat=Stereo
 sampleRate=48000
}
