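# Stereo MetaSound source for the looping first-person in-air wind audio of a closed glider, using skydive wind recordings.
# It is a preset of Glider_InAirAudio_Loop_Parent_MSS, which holds the graph logic, and supplies cloth flap, horizontal, vertical and sideways loop waves.
# Float inputs for forward, back, left, right and ground proximity feed the parent, with forward high by default.
# It plays through the Fort_Gliders_InAir_1P sound class with glider open/close attenuation.
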MetaSoundSource /Game/Athena/Sounds/Wind_Falling/Glider_InAirAudio_Loop_Closed_MSS
preset_parent=&/Glider_InAirAudio_Loop_Parent_MSS
inputs [
 BackParam {type=Float}
 ClothFlapLoop {type=WaveAsset}
 ForwardParam {type=Float}
 GroundProx {type=Float}
 HorizontalLoop {type=WaveAsset}
 LeftParam {type=Float}
 RightParam {type=Float}
 SidewaysLoop {type=WaveAsset}
 UE.Source.OnPlay {type=Trigger}
 VerticalLoop {type=WaveAsset}
]
interfaces [UE.OutputFormat.Stereo:1.0, UE.Source:1.1]
outputs [
 "UE.OutputFormat.Stereo.Audio:0" {type=Audio}
 "UE.OutputFormat.Stereo.Audio:1" {type=Audio}
]
pages [Default {input_defaults {
   BackParam=0
   ClothFlapLoop=&/AMB_Wind_Skydive_Cloth_Flap_Loop_01
   ForwardParam=0.876
   GroundProx=0
   HorizontalLoop=&/AMB_Wind_Skydive_Horizontal_Loop_01
   LeftParam=0
   RightParam=0
   SidewaysLoop=&/AMB_Wind_Skydive_Bank_Loop_01
   UE.Source.OnPlay=null
   VerticalLoop=&/AMB_Wind_Skydive_Vertical_Loop_01
  }}]
properties {
 assetImportData=&self:AssetImportData
 attenuationSettings=/Game/Sounds/Attenuation/Fort_Player_Glider_OpenClose_Athena_Att
 bLooping=true
 duration=10000
 maxDistance=10000
 numChannels=2
 outputFormat=Stereo
 soundClassObject=/Game/Sounds/FortSoundClassesAndModes/Fort_Gliders_InAir_1P
 volume=0.6
}
