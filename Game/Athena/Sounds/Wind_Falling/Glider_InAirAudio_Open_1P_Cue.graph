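# Looping first-person sound cue for the wind heard while airborne with an open glider.
# The first two quality slots use a GroundProx crossfade that fades down the directional wind mix and brings in a parachute skydive loop as the value rises.
# That mix layers left and right bank loops, faded by LeftParam and RightParam, with a horizontal loop volume-modulated by ForwardParam.
# The other two quality slots play only a single umbrella-open wind loop.
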SoundCue /Game/Athena/Sounds/Wind_Falling/Glider_InAirAudio_Open_1P_Cue
root=QualityLevel_0
nodes [
 {id=Mixer_0, children [WavePlayer_0], properties {inputVolume [1]}}
 {id=Mixer_1, children [ParamCrossFade_1, ParamCrossFade_2, ModulatorContinuous_0], properties {inputVolume [1,1,1]}}
 {
  id=ModulatorContinuous_0
  children [WavePlayer_14]
  properties {volumeModulationParams {maxOutput=1.5, minOutput=1, parameterName=ForwardParam}}
 }
 {
  id=ParamCrossFade_0
  children [Mixer_1,Mixer_0]
  properties {
   paramName=GroundProx
   crossFadeInput [
    {fadeInDistanceEnd=0, fadeInDistanceStart=0, fadeOutDistanceEnd=2, fadeOutDistanceStart=0, volume=1}
    {fadeInDistanceEnd=0.4, fadeInDistanceStart=0, fadeOutDistanceEnd=100, fadeOutDistanceStart=1, volume=1}
   ]
  }
 }
 {
  id=ParamCrossFade_1
  children [WavePlayer_12]
  properties {
   paramName=RightParam
   crossFadeInput [{fadeInDistanceEnd=1, fadeInDistanceStart=0, fadeOutDistanceEnd=100, fadeOutDistanceStart=1, volume=1}]
  }
 }
 {
  id=ParamCrossFade_2
  children [WavePlayer_8]
  properties {paramName=LeftParam, crossFadeInput [{fadeInDistanceEnd=1, fadeInDistanceStart=0, fadeOutDistanceEnd=100, fadeOutDistanceStart=1, volume=1}]}
 }
 {id=QualityLevel_0, children [ParamCrossFade_0, ParamCrossFade_0, WavePlayer_7, WavePlayer_7]}
 {id=WavePlayer_0, properties {bLooping=true, soundWaveAssetPtr=&/sfx_parachute_skydive_loop_3p}}
 {id=WavePlayer_12, properties {bLooping=true, soundWaveAssetPtr=&/AMB_Wind_Skydive_Bank_Right_Loop_01}}
 {id=WavePlayer_14, properties {bLooping=true, soundWaveAssetPtr=&/AMB_Wind_Skydive_Horizontal_Loop_01}}
 {id=WavePlayer_7, properties {bLooping=true, soundWaveAssetPtr=&/Wind_UmbrellaOpen_Loop_01_opt}}
 {id=WavePlayer_8, properties {bLooping=true, soundWaveAssetPtr=&/AMB_Wind_Skydive_Bank_Left_Loop_01}}
]
properties {
 attenuationSettings=/Game/Sounds/Attenuation/Fort_Player_Glider_OpenClose_Athena_Att
 maxDistance=10000
 soundClassObject=/Game/Sounds/FortSoundClassesAndModes/Fort_Gliders_InAir_1P
 volumeMultiplier=0.15
}
