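# Third-person looping in-air sound cue for an open glider in Athena, routed to the glider in-air sound class.
# Mixes three looping waves, each with its own attenuation: a close-detail skydive loop plus close and distant synth layers pitched up by fixed modulators.
# A root enveloper fades the whole mix in from silence.
# A concurrency override caps simultaneous instances and scales their volume by distance.
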SoundCue /Game/Athena/Sounds/Wind_Falling/Glider_InAirAudio_Open_3P_Cue
root=Enveloper_0
nodes [
 {
  id=Attenuation_0
  children [Modulator_1]
  properties {attenuationSettings=/Game/Sounds/Attenuation/Fort_Player_Glider_GenericInAir_Close_Athena_Att}
 }
 {
  id=Attenuation_1
  children [WavePlayer_3]
  properties {attenuationSettings=/Game/Sounds/Attenuation/Fort_Player_Glider_CloseDetail_Athena_Att}
 }
 {
  id=Attenuation_3
  children [Modulator_0]
  properties {attenuationSettings=/Game/Sounds/Attenuation/Fort_Player_Glider_GenericInAir_Distant_Athena_Att}
 }
 {
  id=Enveloper_0
  children [Mixer_0]
  properties {pitchCurve {editorCurveData {keys [{time=0}]}}, volumeCurve {editorCurveData {keys [<at 0 0>,<at 1 1>]}}}
 }
 {id=Mixer_0, children [Attenuation_1, Mixer_1], properties {inputVolume [2,2]}}
 {id=Mixer_1, children [Attenuation_0, Attenuation_3], properties {inputVolume [2,3]}}
 {id=Modulator_0, children [WavePlayer_5], properties {pitchMax=1.3, pitchMin=1.3}}
 {id=Modulator_1, children [WavePlayer_4], properties {pitchMax=1.3, pitchMin=1.3}}
 {id=WavePlayer_3, properties {bLooping=true, soundWaveAssetPtr=&/sfx_parachute_skydive_loop_3p}}
 {id=WavePlayer_4, properties {bLooping=true, soundWaveAssetPtr=&/sfx_parachute_skydive_loop_synth_3p_02}}
 {
  id=WavePlayer_5
  properties {bLooping=true, soundWaveAssetPtr=&/sfx_parachute_skydive_loop_synth_3p_distant_02}
 }
]
properties {
 bOverrideConcurrency=true
 maxDistance=9000
 soundClassObject=&Fort_Gliders_InAir
 volumeMultiplier=1.5
 attenuationOverrides {bAttenuate=false, falloffDistance=100000}
 concurrencyOverrides {maxCount=4, voiceStealReleaseTime=0.15, volumeScale=0.15, volumeScaleAttackTime=0.5, volumeScaleMode=Distance}
}
