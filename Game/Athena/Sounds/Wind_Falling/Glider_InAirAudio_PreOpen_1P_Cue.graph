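# Sound cue for the first-person in-air wind gust heard during skydiving before the glider opens in Athena.
# A single wave player feeds the root directly with the pre-open skydive gust wave at reduced volume, with no randomization or modulation.
# It uses the glider open/close attenuation and first-person in-air glider sound class.
# A concurrency override allows one instance per owner, stopping the oldest.
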SoundCue /Game/Athena/Sounds/Wind_Falling/Glider_InAirAudio_PreOpen_1P_Cue
root=WavePlayer_7
nodes [{id=WavePlayer_7, properties {soundWaveAssetPtr=&/AMB_Wind_Skydive_Gust_PreOpen_01}}]
properties {
 attenuationSettings=/Game/Sounds/Attenuation/Fort_Player_Glider_OpenClose_Athena_Att
 bOverrideConcurrency=true
 maxDistance=10000
 soundClassObject=/Game/Sounds/FortSoundClassesAndModes/Fort_Gliders_InAir_1P
 volumeMultiplier=0.4
 concurrencyOverrides {bLimitToOwner=true, maxCount=1, resolutionRule=StopOldest}
}
